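# Translucent, unlit, two-sided gradient particle material for Niagara sprites and ribbons, filed under the monolith connecting-beam slam effects.
# An offset V coordinate, shaped by contrast, inversion and two powers, blends BrightColor and DarkCOlor, then is tinted by particle color and passed through MF_FX_EmissiveAdjustments.
# Opacity multiplies particle alpha, a shape texture with scale and pan controls, and a powered depth fade that softens intersections.
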Material /PistolCanary/Assets/FX/Monoliths/ConnectingBeams/Slam/M_GradParticle
graph {
 properties {
  bAutomaticallySetUsageInEditor=false
  bUsedWithNiagaraRibbons=true
  bUsedWithNiagaraSprites=true
  blendMode=BLEND_Translucent
  shadingModel=MSM_Unlit
  twoSided=true
  usedShadingModels=Unlit
 }
 nodes [
  n1 = Add(
   A=MaterialFunctionCall(Float2=TextureCoordinate()) {materialFunction=&BreakOutFloat2Components}.G
   B=ScalarParameter() {defaultValue=0.0256, parameterName=Offset}
  )
  n2 = ScalarParameter() {parameterName=Contrast}
  n3 = Power(
   Base=OneMinus(None=MaterialFunctionCall(In=@n1, Contrast=@n2) {materialFunction=&CheapContrast}.Result)
   Exp=ScalarParameter() {defaultValue=5.0979, parameterName=Power2}
  )
  n4 = LinearInterpolate(
   A=VectorParameter() {parameterName=BrightColor, defaultValue=<rgba 20 20 20 1>}.RGB
   B=VectorParameter() {parameterName=DarkCOlor, defaultValue=<rgba 0.05 0.05 0.05 1>}.RGB
   Alpha=Power(Base=OneMinus(None=@n3), Exp=ScalarParameter() {defaultValue=4.1028, parameterName=Power})
  )
  n5 = ParticleColor()
  MaterialFunctionCall(Emissive=Divide(A=Multiply(A=@n4, B=@n5.RGB)) {constB=2.271}) {materialFunction=&MF_FX_EmissiveAdjustments}
  n6 = TextureObjectParameter() {
   group=ParticleTex
   parameterName=ParticleTex_Texture
   sortPriority=0
   texture=&/ParticleShape
  }
  n7 = VectorParameter() {group=ParticleTex, parameterName=ParticleTex_UVControl, channelNames {a=VPan, b=UPan, g=VScale, r=UScale}, defaultValue=<rgba 1 1 0 0>}
  n8 = ScalarParameter() {parameterName=DF}
  Multiply(
   A=Multiply(
    A=@n5.A
    B=MaterialFunctionCall(Texture=@n6, UVControl_RGB=@n7, UVControl_A=@n7) {materialFunction=&FXCommon_TextureSample}.R
   )
   B=Power(Base=DepthFade(FadeDistance=@n8), Exp=ScalarParameter() {defaultValue=2, parameterName=DFPower})
  )
 ]
}
